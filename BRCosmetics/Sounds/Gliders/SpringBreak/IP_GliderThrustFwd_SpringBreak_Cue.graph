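# SoundCue for the SpringBreak glider's forward thrust, pairing thrust effects with a musical theme.
# A mixer plays one of three randomly chosen thrust one-shots alongside a music intro, both music layers boosted relative to the thrust.
# A fixed delay of about two and a half seconds starts the looping music track after the intro begins.
# It uses glider thrust attenuation and concurrency and routes to the glider music sound class.
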SoundCue /BRCosmetics/Sounds/Gliders/SpringBreak/IP_GliderThrustFwd_SpringBreak_Cue
root=Mixer_0
nodes [
 {id=Delay_0, children [WavePlayer_4], properties {delayMax=2.49, delayMin=2.49}}
 {id=Mixer_0, children [Random_0,WavePlayer_3,Delay_0], properties {inputVolume [1,1.5,1.5]}}
 {id=Random_0, children [WavePlayer_0,WavePlayer_1,WavePlayer_2], properties {weights [1,1,1]}}
 {id=WavePlayer_0, properties {soundWaveAssetPtr=&/IP_Glider_SpringBreak_FwdThrust_01}}
 {id=WavePlayer_1, properties {soundWaveAssetPtr=&/IP_Glider_SpringBreak_FwdThrust_03}}
 {id=WavePlayer_2, properties {soundWaveAssetPtr=&/IP_Glider_SpringBreak_FwdThrust_02}}
 {id=WavePlayer_3, properties {soundWaveAssetPtr=&/IP_Glider_SpringBreak_Music_Intro}}
 {id=WavePlayer_4, properties {bLooping=true, soundWaveAssetPtr=&/IP_Glider_SpringBreak_Music_LP}}
]
properties {
 attenuationSettings=&Fort_Player_Glider_Thrust_Athena_Att
 audioPropertiesSheet=&APS_Cosmetic_Glider_Loop
 maxDistance=12500
 soundClassObject=/Game/Sounds/FortSoundClassesAndModes/Fort_Gliders_Music
 volumeMultiplier=0.55
 concurrencySet [&Athena_Glider_Thrust_Conc]
}
