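# Sound cue for the SpringBreak glider's looping thrust audio, layering continuous movement loops with intermittent groans.
# Water, low water, wind and sail loops have pitch or volume driven by the AnyDirection parameter, with low water volume mapped inversely to fade as it rises.
# Two looping branches add random groans from separate four-sound pools after random delays, each through a Modulator, mixed with the thrust bed under glider attenuation.
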SoundCue /BRCosmetics/Sounds/Gliders/SpringBreak/IP_GliderThrustLoop_SpringBreak_Cue
root=Mixer_0
nodes [
 {id=Delay_0, children [Random_1], properties {delayMax=5, delayMin=1}}
 {id=Delay_1, children [Random_2], properties {delayMax=6, delayMin=3}}
 {id=Looping_0, children [Modulator_0]}
 {id=Looping_1, children [Modulator_1]}
 {id=Mixer_0, children [Mixer_2,Mixer_1], properties {inputVolume [0.75,1.2]}}
 {id=Mixer_1, children [Looping_0,Looping_1], properties {inputVolume [1.25,1.25]}}
 {
  id=Mixer_2
  children [ModulatorContinuous_0, ModulatorContinuous_1, ModulatorContinuous_2, ModulatorContinuous_3]
  properties {inputVolume [1,1,0.8,1]}
 }
 {
  id=ModulatorContinuous_0
  children [WavePlayer_10]
  properties {
   pitchModulationParams {maxOutput=1.2, minOutput=1, parameterName=AnyDirection}
   volumeModulationParams {maxOutput=0.8, minOutput=0.25, parameterName=AnyDirection}
  }
 }
 {
  id=ModulatorContinuous_1
  children [WavePlayer_11]
  properties {
   pitchModulationParams {maxOutput=1.1, minOutput=1, parameterName=AnyDirection}
   volumeModulationParams {maxOutput=0.6, minOutput=0.25, parameterName=AnyDirection}
  }
 }
 {
  id=ModulatorContinuous_2
  children [WavePlayer_9]
  properties {
   pitchModulationParams {minOutput=1, parameterName=AnyDirection}
   volumeModulationParams {maxOutput=0.1, minOutput=0.5, parameterName=AnyDirection}
  }
 }
 {
  id=ModulatorContinuous_3
  children [WavePlayer_8]
  properties {
   pitchModulationParams {maxOutput=1.1, minOutput=1, parameterName=AnyDirection}
   volumeModulationParams {minOutput=0.6, parameterName=AnyDirection}
  }
 }
 {id=Modulator_0, children [Delay_0]}
 {id=Modulator_1, children [Delay_1]}
 {id=Random_1, children [WavePlayer_0,WavePlayer_1,WavePlayer_2,WavePlayer_3], properties {weights [1,1,1,1]}}
 {id=Random_2, children [WavePlayer_4,WavePlayer_5,WavePlayer_6,WavePlayer_7], properties {weights [1,1,1,1]}}
 {id=WavePlayer_0, properties {soundWaveAssetPtr=&/IP_Glider_SpringBreak_Groan_01}}
 {id=WavePlayer_1, properties {soundWaveAssetPtr=&/IP_Glider_SpringBreak_Groan_02}}
 {
  id=WavePlayer_10
  properties {bLooping=true, soundWaveAssetPtr=&/IP_Glider_SpringBreak_Thrust_Water_LP}
 }
 {id=WavePlayer_11, properties {bLooping=true, soundWaveAssetPtr=&/IP_Glider_SpringBreak_Thrust_Wind_LP}}
 {id=WavePlayer_2, properties {soundWaveAssetPtr=&/IP_Glider_SpringBreak_Groan_03}}
 {id=WavePlayer_3, properties {soundWaveAssetPtr=&/IP_Glider_SpringBreak_Groan_04}}
 {id=WavePlayer_4, properties {soundWaveAssetPtr=&/IP_Glider_SpringBreak_Groan_05}}
 {id=WavePlayer_5, properties {soundWaveAssetPtr=&/IP_Glider_SpringBreak_Groan_06}}
 {id=WavePlayer_6, properties {soundWaveAssetPtr=&/IP_Glider_SpringBreak_Groan_07}}
 {id=WavePlayer_7, properties {soundWaveAssetPtr=&/IP_Glider_SpringBreak_Groan_08}}
 {id=WavePlayer_8, properties {bLooping=true, soundWaveAssetPtr=&/IP_Glider_SpringBreak_Thrust_Sails_LP}}
 {
  id=WavePlayer_9
  properties {bLooping=true, soundWaveAssetPtr=&/IP_Glider_SpringBreak_Thrust_Water_Low_LP}
 }
]
properties {
 attenuationSettings=&Fort_Player_Glider_Thrust_Athena_Att
 audioPropertiesSheet=&APS_Cosmetic_Glider_Loop
 maxDistance=12500
 soundClassObject=&Fort_Gliders_Land
 volumeMultiplier=0.45
 concurrencySet [null]
}
