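# Sound cue for the cannon-fire effect of the SpringBreak glider cosmetic.
# A Random node picks one of three equally weighted CannonFire waves and feeds a Modulator left at default settings.
# It uses the glider open/close attenuation asset and concurrency set, the glider landing sound class, and reduced volume.
# The debug flag is left enabled on the cue.
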SoundCue /BRCosmetics/Sounds/Gliders/SpringBreak/IP_Glider_SpringBreak_CannonFire_cue
root=Modulator_0
nodes [
 {id=Modulator_0, children [Random_1]}
 {id=Random_1, children [WavePlayer_2,WavePlayer_3,WavePlayer_4], properties {weights [1,1,1]}}
 {id=WavePlayer_2, properties {soundWaveAssetPtr=&/IP_Glider_SpringBreak_CannonFire_03}}
 {id=WavePlayer_3, properties {soundWaveAssetPtr=&/IP_Glider_SpringBreak_CannonFire_01}}
 {id=WavePlayer_4, properties {soundWaveAssetPtr=&/IP_Glider_SpringBreak_CannonFire_02}}
]
properties {
 attenuationSettings=/Game/Sounds/Attenuation/Fort_Player_Glider_OpenClose_Athena_Att
 bDebug=true
 maxDistance=10000
 soundClassObject=&Fort_Gliders_Land
 volumeMultiplier=0.6
 attenuationOverrides {falloffDistance=100, stereoSpread=50, attenuationShapeExtents {x=10}}
 concurrencySet [&Athena_GliderOpenClose_Conc]
}
